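; CalendarDataSoundComponent changes its owner's audio as calendar events activate, using event-name maps for sound, attenuation, float and int values.
; Unless the owner watches the calendar, it gathers map keys and watches them after playlist and calendar sync.
; Name or index events apply mapped values on cosmetic-capable instances and fire dispatchers.
; Sound changes reuse the owner's AudioComponent or spawn attached audio; TestCalendarDataChange simulates events in PIE.
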(event EventBeginPlay
  (bind _returnvalue (Components|GetOwner))
  (bind _returnvalue_2 (Actor|GetComponentbyClass _returnvalue "/Script/Engine.AudioComponent"))
  (bind _calendarevents (Variables|Advanced|GetCalendarEvents))
  ;[0.04308 0.08917 0.15 0.5] Set Up Calendar Events and Begin Watching
  (Utilities|FlowControl|Sequence
    (:then_0
      (Variables|Configuration|GetUseOwnerToWatchCalendar
        (:then)
        (:else
          (CallFunction|SetCalendarArraytoMapInputs)
          (Class|FortAsyncActionCurrentPlaylistReady|PlaylistDataReadyAsync
            (:OnReady
              (MCPContext|InitialCalendarSyncCompleted
                (:OnSyncComplete
                  (if (Utilities|Array|IsNotEmpty _calendarevents)
                    (bind _eventname (MCPContext|WatchCalendarEvents _calendarevents)
                      (:OnEventActive
                        (CallFunction|CalendarEventActivated _eventname)))))))))))
    (:then_1
      (Variables|Configuration|GetUseOwnerAudioComponent
        (:then
          (Utilities|IsValid _returnvalue_2
            (:Is%20Valid
              (Variables|Advanced|SetAudioComponent _returnvalue_2))
            (:Is%20Not%20Valid
              (Development|PrintText "NSLOCTEXT(\"[A89D584B68966E2EBB33DBC20839A767]\", \"EAEA53BD4A4A99E0CC73B39BC0745ED7\", \"Warning: CalendarDataSoundComponent has bUseOwnerAudioComponent as true,  but the component\\'s owner has no audio component.\")"))))
        (:else)))))

(event Custom|CalendarEventActivated
  (params (CalendarEventName :type String))
  (bind _calendardatatofloatmap (Variables|Configuration|GetCalendarDataToFloatMap))
  (bind (_value _returnvalue) (Utilities|Map|Find _calendardatatofloatmap CalendarEventName))
  (bind _calendardatatointmap (Variables|Configuration|GetCalendarDataToIntMap))
  (bind (_value_2 _returnvalue_2) (Utilities|Map|Find _calendardatatointmap CalendarEventName))
  (bind _calendardatatosoundmap (Variables|Configuration|GetCalendarDataToSoundMap))
  (bind (_value_3 _returnvalue_3) (Utilities|Map|Find _calendardatatosoundmap CalendarEventName))
  (bind _calendardatatoattenuationmap (Variables|Configuration|GetCalendarDataToAttenuationMap))
  (bind (_value_4 _returnvalue_4) (Utilities|Map|Find _calendardatatoattenuationmap CalendarEventName))
  ;[0.15 0.05009 0.1189 0.5] Execute Changes When Calendar Event is Triggered
  (Variables|Advanced|GetShouldExecuteMapLogic
    (:then
      (Networking|CanExecuteCosmeticEvents
        (:True
          (Default|CallOnCalendarEventActivated)
          (Utilities|FlowControl|Sequence
            (:then_0
              (if (and (Utilities|Map|IsNotEmpty _calendardatatosoundmap) _returnvalue_3)
                (CallFunction|OnCalendarSoundChanged _value_3)
                (else
                  (Utilities|FlowControl|DelayUntilNextTick)))
              (if (and (Utilities|Map|IsNotEmpty _calendardatatoattenuationmap) _returnvalue_4)
                (CallFunction|OnCalendarAttenuationChanged _value_4)))
            (:then_1
              (if (and (Utilities|Map|IsNotEmpty _calendardatatofloatmap) _returnvalue)
                (CallFunction|OnCalendarFloatChanged _value)))
            (:then_2
              (if (and (Utilities|Map|IsNotEmpty _calendardatatointmap) _returnvalue_2)
                (CallFunction|OnCalendarIntChanged _value_2)))))))
    (:else)))

(event Custom|CalendarDataChangedByInt
  (params (CalendarPhaseInt :type Integer))
  (bind _calendarevents (Variables|Advanced|GetCalendarEvents))
  (bind _output (Utilities|Array|Get(acopy) _calendarevents CalendarPhaseInt))
  (bind _calendardatatointmap (Variables|Configuration|GetCalendarDataToIntMap))
  (bind (_value _returnvalue) (Utilities|Map|Find _calendardatatointmap _output))
  (bind _calendarevents_2 (Variables|Advanced|GetCalendarEvents))
  (bind _output_2 (Utilities|Array|Get(acopy) _calendarevents_2 CalendarPhaseInt))
  (bind _calendardatatosoundmap (Variables|Configuration|GetCalendarDataToSoundMap))
  (bind (_value_2 _returnvalue_2) (Utilities|Map|Find _calendardatatosoundmap _output_2))
  (bind _calendarevents_3 (Variables|Advanced|GetCalendarEvents))
  (bind _output_3 (Utilities|Array|Get(acopy) _calendarevents_3 CalendarPhaseInt))
  (bind _calendardatatoattenuationmap (Variables|Configuration|GetCalendarDataToAttenuationMap))
  (bind (_value_3 _returnvalue_3) (Utilities|Map|Find _calendardatatoattenuationmap _output_3))
  (bind _calendarevents_4 (Variables|Advanced|GetCalendarEvents))
  (bind _output_4 (Utilities|Array|Get(acopy) _calendarevents_4 CalendarPhaseInt))
  (bind _calendardatatofloatmap (Variables|Configuration|GetCalendarDataToFloatMap))
  (bind (_value_4 _returnvalue_4) (Utilities|Map|Find _calendardatatofloatmap _output_4))
  ;[0.02239 0.15 0.1122 0.5] Execute Changes on Phase Integer Changed
  (Variables|Advanced|GetShouldExecuteMapLogic
    (:then
      (Networking|CanExecuteCosmeticEvents
        (:True
          (Default|CallOnCalendarEventActivated)
          (Utilities|FlowControl|Sequence
            (:then_0
              (if (and (Utilities|Map|IsNotEmpty _calendardatatosoundmap) _returnvalue_2)
                (CallFunction|OnCalendarSoundChanged _value_2)
                (else
                  (Utilities|FlowControl|DelayUntilNextTick)))
              (if (and (Utilities|Map|IsNotEmpty _calendardatatoattenuationmap) _returnvalue_3)
                (CallFunction|OnCalendarAttenuationChanged _value_3)))
            (:then_1
              (if (and (Utilities|Map|IsNotEmpty _calendardatatofloatmap) _returnvalue_4)
                (CallFunction|OnCalendarFloatChanged _value_4)))
            (:then_2
              (if (and (Utilities|Map|IsNotEmpty _calendardatatointmap) _returnvalue)
                (CallFunction|OnCalendarIntChanged _value)))))))
    (:else)))

(event Custom|TestCalendarDataChange
  (params (DC%20Event%20Name :type String))
  (bind _calendarevents (Variables|Advanced|GetCalendarEvents))
  (bind _returnvalue (Utilities|Array|FindItem _calendarevents DC%20Event%20Name))
  ;[0.15 0.133 0.06181 0.5] To Test in PIE: cheat ke * TestCalendarDataChange [CalendarEventNameString]
  (if (Utilities|Array|ContainsItem (Variables|Advanced|GetCalendarEvents) DC%20Event%20Name)
    (CallFunction|CalendarDataChangedbyInt _returnvalue)
    (Development|PrintString (Utilities|String|Append "TESTING CALENDAR EVENT: " (Utilities|Array|Get(acopy) (Variables|Advanced|GetCalendarEvents) _returnvalue)) true true "(R=0.000000,G=0.660000,B=1.000000,A=1.000000)" 5.0)
    (else
      (Development|PrintString (Utilities|String|Append DC%20Event%20Name "Not Found In Mapped Events") true true "(R=0.000000,G=0.660000,B=1.000000,A=1.000000)" 5.0))))

(fn OnCalendarAttenuationChanged
  (params (NewCalendarAttenuation :type Sound%20Attenuation%20Object%20Reference))
  (bind _audiocomponent (Variables|Advanced|GetAudioComponent)
    (:then
      (Variables|/Script/Engine.AudioComponent|SetAttenuationSettings (Variables|GetNewCalendarAttenuation) _audiocomponent)
      (Default|CallOnDataCalendarAttenuationChanged (Variables|GetNewCalendarAttenuation)))
    (:else)))

(fn OnCalendarFloatChanged
  (params (NewCalendarFloat :type Float%20%28double-precision%29))
  (Default|CallOnDataCalendarFloatChanged NewCalendarFloat))

(fn OnCalendarIntChanged
  (params (NewCalendarInt :type Integer))
  (Default|CallOnDataCalendarIntChanged NewCalendarInt))

(fn OnCalendarSoundChanged
  (params (NewCalendarSound :type Sound%20Base%20Object%20Reference))
  (Variables|Configuration|GetUseOwnerAudioComponent
    (:then
      (bind _audiocomponent (Variables|Advanced|GetAudioComponent)
        (:then
          (Audio|Components|Audio|FadeOut _audiocomponent 0.5 0.0 "Linear" :!class /Script/Engine.AudioComponent)
          (Audio|Components|Audio|SetSound _audiocomponent (Variables|GetNewCalendarSound))
          (Audio|Components|Audio|Play _audiocomponent 0.0))
        (:else)))
    (:else
      (bind _audiocomponent_2 (Variables|Advanced|GetAudioComponent)
        (:then
          (Audio|Components|Audio|FadeOut _audiocomponent_2 0.5 0.0 "Linear" :!class /Script/Engine.AudioComponent))
        (:else
          _))
      (bind _returnvalue (Audio|SpawnSoundAttached (Variables|GetNewCalendarSound) (Variables|/Script/Engine.Actor|GetRootComponent (Components|GetOwner))))
      (Variables|Advanced|SetAudioComponent _returnvalue)))
  (Default|CallOnDataCalendarSoundChanged (Variables|GetNewCalendarSound)))

(fn SetCalendarArrayToMapInputs
  (bind _calendardatatointmap (Variables|Configuration|GetCalendarDataToIntMap))
  (bind _returnvalue (Utilities|Map|Length _calendardatatointmap))
  (bind _calendardatatoattenuationmap (Variables|Configuration|GetCalendarDataToAttenuationMap))
  (bind _returnvalue_2 (Utilities|Map|Length _calendardatatoattenuationmap))
  (bind _calendardatatosoundmap (Variables|Configuration|GetCalendarDataToSoundMap))
  (bind _returnvalue_3 (Utilities|Map|Length _calendardatatosoundmap))
  (bind _calendardatatofloatmap (Variables|Configuration|GetCalendarDataToFloatMap))
  (bind _returnvalue_4 (Utilities|Map|Length _calendardatatofloatmap))
  (bind _calendardatatofloatmap_2 (Variables|Configuration|GetCalendarDataToFloatMap))
  (bind _calendardatatointmap_2 (Variables|Configuration|GetCalendarDataToIntMap))
  (bind _calendardatatoattenuationmap_2 (Variables|Configuration|GetCalendarDataToAttenuationMap))
  (bind _calendardatatosoundmap_2 (Variables|Configuration|GetCalendarDataToSoundMap))
  (if (> _returnvalue_3 0)
    (Variables|Advanced|SetShouldExecuteMapLogic true)
    (for _index (range _returnvalue_3)
      (if (not (Utilities|Array|ContainsItem (Variables|Advanced|GetCalendarEvents) (Utilities|Map|GetKeyValueByIndex _calendardatatosoundmap_2 _index)))
        (Utilities|Array|Add (Variables|Advanced|GetCalendarEvents) (Utilities|Map|GetKeyValueByIndex _calendardatatosoundmap_2 _index))))
    (else
      _))
  (if (> _returnvalue_4 0)
    (Variables|Advanced|SetShouldExecuteMapLogic true)
    (for _index_2 (range _returnvalue_4)
      (if (not (Utilities|Array|ContainsItem (Variables|Advanced|GetCalendarEvents) (Utilities|Map|GetKeyValueByIndex _calendardatatofloatmap_2 _index_2)))
        (Utilities|Array|Add (Variables|Advanced|GetCalendarEvents) (Utilities|Map|GetKeyValueByIndex _calendardatatofloatmap_2 _index_2))))
    (else
      _))
  (if (> _returnvalue 0)
    (Variables|Advanced|SetShouldExecuteMapLogic true)
    (for _index_3 (range _returnvalue)
      (if (not (Utilities|Array|ContainsItem (Variables|Advanced|GetCalendarEvents) (Utilities|Map|GetKeyValueByIndex _calendardatatointmap_2 _index_3)))
        (Utilities|Array|Add (Variables|Advanced|GetCalendarEvents) (Utilities|Map|GetKeyValueByIndex _calendardatatointmap_2 _index_3))))
    (else
      _))
  (if (> _returnvalue_2 0)
    (Variables|Advanced|SetShouldExecuteMapLogic true)
    (for _index_4 (range _returnvalue_2)
      (if (not (Utilities|Array|ContainsItem (Variables|Advanced|GetCalendarEvents) (Utilities|Map|GetKeyValueByIndex _calendardatatoattenuationmap_2 _index_4)))
        (Utilities|Array|Add (Variables|Advanced|GetCalendarEvents) (Utilities|Map|GetKeyValueByIndex _calendardatatoattenuationmap_2 _index_4))))
    (return)
    (else
      (return))))
